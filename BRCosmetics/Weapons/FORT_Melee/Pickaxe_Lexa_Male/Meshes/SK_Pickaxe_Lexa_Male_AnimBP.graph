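# Animation Blueprint for the Lexa Male pickaxe that makes its dyn_element bone, apparently a shard, float procedurally.
# Each animation update builds the ShardFloatA vector from game-time sine waves at a different frequency per axis, with the Y wave squared so it bobs to one side only.
# The AnimGraph applies that vector to the input pose as an additive bone-space translation through a Modify Bone node.
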AnimBlueprint /BRCosmetics/Weapons/FORT_Melee/Pickaxe_Lexa_Male/Meshes/SK_Pickaxe_Lexa_Male_AnimBP
target_skeleton=&/SK_Pickaxe_Lexa_Male_Skeleton
graphs [
 AnimGraph {nodes [
   AnimGraphNode_SubInput_0 = Input Pose()
   n1 = "Transform (Modify) Bone - Bone: dyn_element"(
    ComponentPose=Local To Component(LocalPose=@AnimGraphNode_SubInput_0)
    Alpha=1
    Translation=Get ShardFloatA() -> (ShardFloatA:Vector)
    Rotation="0,0,0"
    Scale="1,1,1"
   ) {translationMode=BMM_Additive, translationSpace=BCS_BoneSpace, alphaBoolBlend {blendOption=Linear}, boneToModify {boneName=dyn_element}}
   Output Pose(Result=Component To Local(ComponentPose=@n1)) {layerGroup=None}
  ]}
 EventGraph {nodes [
   Event BlueprintUpdateAnimation()
   n1 = float * float(A=1, B=1)
   n2 = GetGameTimeInSeconds()
   n3 = Square!(A:double="Sin (Radians)"(A=float * float(A=@n1, B=@n2))) -> (ReturnValue:double)
   n4 = "Sin (Radians)"(A=float * float(A=float * float(A=1, B=2), B=@n2))
   n5 = "Sin (Radians)"(A=float * float(A=float * float(A=1, B=1.5), B=@n2))
   > Set ShardFloatA(ShardFloatA:Vector=MakeVector(X=float * float(A=@n5, B=-0.4), Y=float * float(A=@n3, B=-0.5), Z=float * float(A=@n4, B=-0.25))) -> (Output_Get:Vector)
   TryGetPawnOwner()
  ]}
]
variables [ShardFloatA {type=Vector}]
