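# Physics asset for the Lexa Male pickaxe skeletal mesh, a two-handed melee harvesting tool cosmetic.
# It holds a single body on the Melee_TwoHanded_Handle bone, shaped as one capsule that serves as a coarse collision volume for the weapon.
# With only that body, it defines no constraints, so the pickaxe has no articulated physics parts.
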PhysicsAsset /BRCosmetics/Weapons/FORT_Melee/Pickaxe_Lexa_Male/Meshes/SK_Pickaxe_Lexa_Male_PhysicsAsset
bodies {Melee_TwoHanded_Handle [{shapeType=Capsule, length=72.0107, radius=33.9252, center=<0.7986 -0.6786 23.1205>, rotation=<rot 1.3041 -90 -0.0089>}]}
constraints []
